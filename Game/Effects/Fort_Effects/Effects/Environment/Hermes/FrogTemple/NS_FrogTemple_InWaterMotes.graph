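# Niagara system for dense drifting motes in the Frog Temple environment, named as an in-water effect.
# The active GPU emitter continuously spawns short-lived, green-tinted sprites in a wide cylinder offset below the origin.
# Particles grow and shrink over life while gravity, turbulent wind and drag move them under a speed limit.
# A second sprite-and-mesh emitter is disabled, as are both emitters' indoor-player-driven spawn rates.
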NiagaraSystem /Game/Effects/Fort_Effects/Effects/Environment/Hermes/FrogTemple/NS_FrogTemple_InWaterMotes
dependencies {
 usedDataInterfaces [&NiagaraDataInterfaceCurve, &NiagaraDataInterfaceCollisionQuery, &NiagaraDataInterfaceLandscape]
 usedDynamicInputs [&SimulationPosition, &Vector2DFromFloat]
 usedModules [
  &SystemState
  &self:SystemSpawnScript.SystemScriptSource.SystemScriptGraph.NiagaraNodeAssignment_0.SetVariables_#1
 ]
}
emitters {
 Dust_Motes {
  bEnabled=false
  simTarget=GPUComputeSim
  emitterUpdateScript [
   &EmitterState(
    Life Cycle Mode=::NewEnumerator1
    Loop Duration=&RandomRangeFloat(Minimum=5, Maximum=10)
    Recalculate Duration Each Loop=true
    UseLoopDelay=false
   )
   &SpawnRate(SpawnRate=&Lerp_Float(A=0, B=100, Alpha=$NPC.FortniteNPC.IsPlayerIndoors), Use Spawn Probability=false) {enabled=false}
   &SpawnRate(SpawnRate=5000, Use Spawn Probability=false) {module=SpawnRate001}
  ]
  particleSpawnScript [
   &InitializeParticle(
    Lifetime=&UniformRangedFloat(Minimum=3, Maximum=5)
    Mass Mode=::NewEnumerator1
    Mesh Scale Max=<0.5 0.5 0.5>
    Mesh Scale Min=<0.05 0.1 0.1>
    Mesh Scale Mode=::NewEnumerator2
    Sprite Rotation Mode=::NewEnumerator2
    Sprite Size Mode=::NewEnumerator3
    Uniform Sprite Size=1
    UsePositionOffset=false
    Write Material Random=true
   ) {module=InitializeParticle001}
   /SetVariables(Particles.MeshIndex=&RandomRangeInt(Minimum=0, Maximum=3)) {isSetParameters=true, module=SetVariables_#2}
   &ShapeLocation(
    Cylinder Height=200
    Cylinder Radius=2500
    MaskSpawnGroup=false
    Offset=<0 0 -150>
    Offset Mode=::NewEnumerator0
    Override Seed=false
    Shape Primitive=::NewEnumerator1
    Surface Only=false
    UseOwnerScale=true
   )
   &AddVelocity(Velocity=&RandomVector(Vector Scale=&RandomRangeFloat(Minimum=0.5, Maximum=1)))
   &DynamicMaterialParameters(
    Index 0 Param 1=2
    Index 0 Param 2=50
    Index 0 Param 3=200
    Module.Param0WriteEnabled=true
    Module.Param1WriteEnabled=true
    Module.Param2WriteEnabled=true
    Module.Param3WriteEnabled=true
   ) {module=DynamicMaterialParameters001}
   &InitialMeshOrientation(Use Rotation Vector=false)
  ]
  particleUpdateScript [
   &ParticleState()
   &ScaleMeshSize(Scale Factor=&VectorFromFloat(Value=&FloatFromCurve(FloatCurve=DataInterface() {propertyValues {Curve {keys [
          {interpMode=RCIM_Cubic, tangentMode=RCTM_User, time=0, value=0}
          {interpMode=RCIM_Cubic, tangentMode=RCTM_User, time=0.3, value=0.95}
          {tangentMode=RCTM_User, time=1, value=0}
         ]}}})))
   &Drag(Drag=0.5, Use Linear Drag=true, Use Rotational Drag=false)
   &GravityForce(Gravity=<0 0 -15>)
   &SpriteRotationRate(Rotation Rate=&RandomRangeFloat(Minimum=-45, Maximum=45))
   &ScaleColor(Scale RGB=<0.5 0.5 0.5>, ScaleA=true, ScaleRGB=true)
   &WindForce(
    Scale=1
    Turbulence Flow Loop Duration=1
    Turbulence Frequency Mode:ENiagaraWindTurbulenceFrequencyMode=::NewEnumerator1
    Wind Speed=<25 0 -25>
   )
   &AerodynamicDrag(Aerodynamic Rotational Drag=0, Rotation Strength=25, Use Linear Drag=true, Use Rotational Drag=true)
   &SolveForcesAndVelocity(Clamp Velocity=true, Limit Acceleration=false, Speed Limit=250)
   &AlignSpriteToMeshOrientation(SetAlignment=true, SetFacing=true)
  ]
  rendererClasses [&NiagaraSpriteRendererProperties, &NiagaraMeshRendererProperties]
  renderers [
   {rendererClass=&NiagaraSpriteRendererProperties}
   {rendererClass=&NiagaraMeshRendererProperties, rendererIndex=1}
  ]
 }
 Dust_Motes001 {
  simTarget=GPUComputeSim
  emitterUpdateScript [
   &EmitterState(
    Life Cycle Mode=::NewEnumerator1
    Loop Duration=&RandomRangeFloat(Minimum=5, Maximum=10)
    Recalculate Duration Each Loop=true
    UseLoopDelay=false
   )
   &SpawnRate(SpawnRate=&Lerp_Float(A=0, B=100, Alpha=$NPC.FortniteNPC.IsPlayerIndoors), Use Spawn Probability=false) {enabled=false}
   &SpawnRate(SpawnRate=5000, Use Spawn Probability=false) {module=SpawnRate001}
  ]
  particleSpawnScript [
   &InitializeParticle(
    Color=<rgba 0.3954 1 0.0906 1>
    Lifetime=&UniformRangedFloat(Minimum=3, Maximum=5)
    Mass Mode=::NewEnumerator1
    Mesh Scale Max=<0.2 0.25 0.25>
    Mesh Scale Min=<0.05 0.1 0.1>
    Mesh Scale Mode=::NewEnumerator2
    Sprite Rotation Mode=::NewEnumerator2
    Sprite Size Mode=::NewEnumerator1
    UsePositionOffset=false
    Write Material Random=true
   ) {module=InitializeParticle001}
   /SetVariables(Particles.MeshIndex=&RandomRangeInt(Minimum=0, Maximum=3)) {isSetParameters=true, module=SetVariables_#2}
   &ShapeLocation(
    Cylinder Radius=2500
    MaskSpawnGroup=false
    Offset=<0 0 -200>
    Offset Mode=::NewEnumerator0
    Override Seed=false
    Shape Primitive=::NewEnumerator1
    Surface Only=false
    UseOwnerScale=true
   )
   &AddVelocity(Velocity=&RandomVector(Vector Scale=&RandomRangeFloat(Minimum=0.5, Maximum=1)))
   &DynamicMaterialParameters(
    Index 0 Param 1=2
    Index 0 Param 2=50
    Index 0 Param 3=200
    Module.Param0WriteEnabled=true
    Module.Param1WriteEnabled=true
    Module.Param2WriteEnabled=true
    Module.Param3WriteEnabled=true
   ) {module=DynamicMaterialParameters001}
   &InitialMeshOrientation(Use Rotation Vector=false)
  ]
  particleUpdateScript [
   &ParticleState()
   &ScaleSpriteSize(Uniform Curve Sprite Scale=DataInterface() {propertyValues {Curve {keys [
        {interpMode=RCIM_Cubic, tangentMode=RCTM_User, time=0, value=0}
        {interpMode=RCIM_Cubic, tangentMode=RCTM_User, time=0.5, value=1}
        {tangentMode=RCTM_User, time=1, value=0}
       ]}}}) {module=ScaleSpriteSize001}
   &ScaleSpriteSize(Uniform Curve Sprite Scale=DataInterface() {propertyValues {Curve {keys [
        {interpMode=RCIM_Cubic, tangentMode=RCTM_User, time=0, value=0}
        {interpMode=RCIM_Cubic, tangentMode=RCTM_User, time=0.5, value=1}
        {tangentMode=RCTM_User, time=1, value=0}
       ]}}})
   &Drag(Drag=0.5, Use Linear Drag=true, Use Rotational Drag=false)
   &GravityForce(Gravity=<0 0 -15>)
   &SpriteRotationRate(Rotation Rate=&RandomRangeFloat(Minimum=-45, Maximum=45))
   &ScaleColor(Scale Alpha=5, Scale RGB=<0.2 0.2 0.2>, ScaleA=true, ScaleRGB=true)
   &WindForce(
    Scale=1
    Turbulence Flow Loop Duration=1
    Turbulence Frequency Mode:ENiagaraWindTurbulenceFrequencyMode=::NewEnumerator1
    Wind Speed=<25 0 -25>
   )
   &AerodynamicDrag(Aerodynamic Rotational Drag=0, Rotation Strength=25, Use Linear Drag=true, Use Rotational Drag=true)
   &SolveForcesAndVelocity(Clamp Velocity=true, Limit Acceleration=false, Speed Limit=250)
   &AlignSpriteToMeshOrientation(SetAlignment=true, SetFacing=true)
  ]
  rendererClasses [&NiagaraSpriteRendererProperties]
  renderers [{rendererClass=&NiagaraSpriteRendererProperties}]
 }
}
summary {
 systemName=NS_FrogTemple_InWaterMotes
 emitters [
  {
   bEnabled=false
   emitterName=Dust_Motes
   simTarget=GPUComputeSim
   rendererClasses [&NiagaraSpriteRendererProperties, &NiagaraMeshRendererProperties]
  }
  {emitterName=Dust_Motes001, simTarget=GPUComputeSim, rendererClasses [&NiagaraSpriteRendererProperties]}
 ]
 userVariables []
}
